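# Niagara system for the VoyagerCuboid cosmetic shoes, adding orange cubes, flash meshes and ribbon trails to both feet, despite its Idle name.
# Idle and running cube emitters swap on the running state; they and the trails stop when skydiving or hidden.
# Ribbon opacity follows speed in gameplay, with separate front-end tuning, while flashes spawn independently of those states.
# Particles use skeletal-mesh sampling and per-foot ShoeScale; legacy ribbon, minimal and debug emitters are disabled.
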NiagaraSystem /CosmeticShoes/Assets/FORT_Shoes/Shoe_VoyagerCuboid/FX/NS_Shoe_VoyagerCuboid_Idle
dependencies {
 usedDataInterfaces [
  &NiagaraDataInterfaceCurve
  &NiagaraDataInterfaceMeshRendererInfo
  &NiagaraDataInterfaceSpriteRendererInfo
  &NiagaraDataInterfaceArrayInt32
  &NiagaraDataInterfaceColorCurve
 ]
 usedDynamicInputs [
  /Game/Effects/Niagara/DynamicInputs/Logic/FN_IsRunningCheck
  &SetBoolByFloatComparison
  &ModuloFloat
 ]
 usedModules [
  &self:SystemSpawnScript.SystemScriptSource.SystemScriptGraph.NiagaraNodeAssignment_0.SetVariables_#1
  &self:SystemSpawnScript.SystemScriptSource.SystemScriptGraph.NiagaraNodeAssignment_4.SetVariables_#2
  &SystemState
  &self:SystemSpawnScript.SystemScriptSource.SystemScriptGraph.NiagaraNodeAssignment_1.SetVariables_#3
  &self:SystemSpawnScript.SystemScriptSource.SystemScriptGraph.NiagaraNodeAssignment_2.SetVariables_#4
  &Timeline
  &self:SystemSpawnScript.SystemScriptSource.SystemScriptGraph.NiagaraNodeAssignment_3.SetVariables_#5
 ]
}
emitters {
 FN_AnimTrail002 {
  emitterSpawnScript [&FN_RibbonIDOffsetUtility()]
  emitterUpdateScript [
   &EmitterState()
   &FN_SpawnAnimTrail(
    Number of Anim Trails:NiagaraInt32=&MakeIntFromFloat(Float=&MakeCustomFloatFromBool(
      Bool=$User.IsSkydiving
      True Float=0
      False Float=&MakeCustomFloatFromBool(
       Bool=$User.HideFootFX
       True Float=0
       False Float=&MakeCustomFloatFromBool(
        Bool=$System.isRunning
        True Float=&MakeCustomFloatFromBool(Bool=$System.isFalling, True Float=0, False Float=2)
        False Float=&MakeCustomFloatFromBool(Bool=$User.bisPrimaryPreviewItem, True Float=2, False Float=0)
       )
      )
     ))
    Switch on ENiagara_AnimTrail_Source:ENiagara_AnimTrail_Source=::NewEnumerator1
   )
   /SetVariables(Emitter.Linear Color:LinearColor=&ColorFromCurve(
     ColorCurve=DataInterface() {propertyValues {
       AlphaCurve {keys []}
       BlueCurve {keys [<at 0 0.7475 RCIM_Cubic>,<at 1 0.283 RCIM_Cubic>]}
       GreenCurve {keys [<at 0 1.8026 RCIM_Cubic>,<at 1 2.3 RCIM_Cubic>]}
       RedCurve {keys [<at 0 4.5 RCIM_Cubic>,<at 1 4.5 RCIM_Cubic>]}
      }}
     CurveIndex=&Waveform("Waveform [1]"=::NewEnumerator0, "[1] Frequency"=0.5, "[1] Amplitude Min/Max"=<0 1>)
    )) {enabled=false, isSetParameters=true}
  ]
  particleSpawnScript [
   &FN_InitializeAnimTrailBoneIndecies(Second Bone or Socket Index=0)
   &FN_InitializeAnimTrailBoneIndecies(AnimTrail Selection=1, First Bone or Socket Index=2, Second Bone or Socket Index=2) {module=FN_InitializeAnimTrailBoneIndecies001}
   &FN_InitializeAnimTrail(
    AnimTrailWidth=&Multiply_Float(
     A=&FN_FrontEndFloatModifier(Gameplay=30, Front End=25)
     B=&MergeHelper_FloatSwitch(
      0=&Multiply_Float(A=&MakeFloatFromVector(Vector=$User.ShoeScale_L, Channel=::NewEnumerator0), B=1)
      1=&Multiply_Float(A=&MakeFloatFromVector(Vector=$User.ShoeScale_R, Channel=::NewEnumerator0), B=1)
      2=?
      3=?
      Index=$Particles.SpawnGroup
      Num=::NewEnumerator0
     )
    )
    Lifetime=&FN_FrontEndFloatModifier(Gameplay=0.085, Front End=0.135)
    SkeletalMesh=$System.OnlyRibbons
   )
   &AddVelocity(Velocity=&MakeVector(
     X=0
     Y=&FN_FrontEndFloatModifier(Gameplay=0, Front End=-150)
     Z=&FloatFromCurve(
      FloatCurve=DataInterface() {propertyValues {Curve {keys [<at 0.0218 1>,<at 0.0388 0>]}}}
      Scale Curve=&FN_FrontEndFloatModifier(Gameplay=70, Front End=0)
     )
    ))
  ]
  particleUpdateScript [
   &SolveForcesAndVelocity(Clamp Velocity=false, Limit Acceleration=false)
   &FN_UpdateAnimTrail(SkeletalMesh=$System.OnlyRibbons) {module=FN_UpdateAnimTrail001}
   /SetVariables(
    Particles.AlphaScale:NiagaraFloat=&FN_FrontEndFloatModifier(
     Gameplay=&RemapRange(
      Input Value To Remap=&VectorLength(Vector=$Output.SolveForcesAndVelocity.Velocity)
      Input Min=400
      Input Max=1000
      Output Min=0
      Output Max=1
     )
     Front End=1
    )
    Particles.GlowColor=<rgba 1 0.3669 0.025 1>
    Particles.LightningColor:LinearColor=<rgba 3 1.0027 0 1>
   ) {isSetParameters=true}
   &DynamicMaterialParameters(
    Index 0 Param 1=0.85
    Index 0 Param 2=1
    Index 0 Param 3=-0.5
    Index 0 Param 4=&MakeCustomFloatFromBool(
     Bool=&Boolean_AndOperation(Bool1=$System.bIsRunning, Bool2=&Boolean_NotOperation(Bool=$User.IsSkydiving))
     True Float=1
     False Float=0
    )
    Module.Index0Float0Write=true
    Module.Index0Float1Write=true
    Module.Index0Float2Write=true
    Module.Index0Float3Write=true
   ) {enabled=false}
   &FN_RibbonMaterial(
    Color=<rgba 1.566 0.5946 0.1843 1>
    Color Pan U Speed=0.1
    Opacity Pan U Speed=1.5
    Opacity Power=0.5
    Opacity Scale U=2
    Scale Alpha:NiagaraFloat=&FloatFromCurve(
     FloatCurve=DataInterface() {propertyValues {Curve {keys [
         {arriveTangent=-1.0872, interpMode=RCIM_Cubic, leaveTangent=-1.0872, tangentMode=RCTM_User, time=0, value=0.5}
         {arriveTangent=-0.3496, interpMode=RCIM_Cubic, leaveTangent=-0.3496, tangentMode=RCTM_User, time=1, value=0}
        ]}}}
     Scale Curve=&Multiply_Float(
      A=$Particles.AlphaScale
      B=&Multiply_Float(A=$User.TrailOn, B=&FNTimeOfDayScalar(Night Scale=0.6, Morning Scale=1, Evening Scale=1))
     )
    )
   )
   &ScaleRibbonWidth(Ribbon Width Scale=&FloatFromCurve(FloatCurve=DataInterface() {propertyValues {Curve {
        defaultValue=1
        keys [
         {interpMode=RCIM_Cubic, tangentMode=RCTM_User, time=0.0057, value=0}
         {arriveTangent=2.4408, interpMode=RCIM_Cubic, leaveTangent=2.4408, time=0.0846, value=0.671}
         {arriveTangent=1.1685, interpMode=RCIM_Cubic, leaveTangent=2.4675, tangentMode=RCTM_Break, time=0.304, value=0.7282}
         {arriveTangent=-3.3834, interpMode=RCIM_Cubic, leaveTangent=-3.3834, tangentMode=RCTM_User, time=0.8306, value=0.3545}
         {arriveTangent=-0.3484, tangentMode=RCTM_User, time=1, value=0}
        ]
       }}})) {enabled=false}
   /SetVariables(Particles.Vector:Vector3f=&ReturnMeshOrientationAxis(Mesh Orientation Axis=::XAxis, Mesh Orientation Quat=$Particles.MeshOrientation)) {isSetParameters=true}
  ]
  rendererClasses [&NiagaraRibbonRendererProperties]
  renderers [
   {rendererClass=&NiagaraRibbonRendererProperties}
   {rendererClass=&NiagaraRibbonRendererProperties, rendererIndex=1}
  ]
 }
 FN_AttachedRibbon001 {
  bEnabled=false
  emitterUpdateScript [
   &EmitterState()
   &SampleSkeletalMesh(
    AdditionalYawPitchRoll=false
    Bone / Socket Index=1
    Bone Sampling Mode:ENiagaraBoneSamplingMode=::NewEnumerator4
    Skeletal Mesh:NiagaraDataInterfaceSkeletalMesh=$System.Skeletal Mesh
   )
   &FN_SpawnRibbon_v2(
    Number of Ribbons=2
    SpawnRate:NiagaraFloat=&MakeCustomFloatFromBool(
     Bool=$User.IsSkydiving
     True Float=0
     False Float=&MakeCustomFloatFromBool(
      Bool=$User.HideFootFX
      True Float=0
      False Float=&MakeCustomFloatFromBool(
       Bool=$System.isRunning
       True Float=&MakeCustomFloatFromBool(Bool=$System.isFalling, True Float=0, False Float=48)
       False Float=&MakeCustomFloatFromBool(
        Bool=$User.bisPrimaryPreviewItem
        True Float=&Multiply_Float(
         A=&ClampFloat(
          Float=&MakeFloatFromVector(Vector=$Emitter.SampleSkeletalMesh.SampledVelocity, Channel=::NewEnumerator1)
          Min=1
         )
         B=48
        )
        False Float=0
       )
      )
     )
    )
   )
  ]
  particleSpawnScript [
   /Game/Effects/Niagara/Modules/Ribbons/FN_InitializeRibbon_v4(
    Enable Skeletal Position=true
    Lifetime:NiagaraFloat=&FN_FrontEndFloatModifier(Gameplay=0.06, Front End=0.1)
    Ribbon Width=500
    SkeletalMesh:NiagaraDataInterfaceSkeletalMesh=$System.OnlyRibbons
   )
   &FN_FlipRibbonV()
   &DynamicMaterialParameters(
    Index 0 Param 1=1
    Index 0 Param 2=1
    Index 0 Param 3=1
    Index 0 Param 4=1
    Index 1 Param 1=1
    Module.Index0Param0Write=true
    Module.Index0Param1Write=true
    Module.Index0Param2Write=true
    Module.Index0Param3Write=true
    Module.Index1Param0Write=true
    Module.Index1Param1Write=true
    Module.Index1Param2Write=true
    Module.Index1Param3Write=true
    Write Parameter Index 1=true
   )
   /SetVariables(Particles.RibbonWidth=&Multiply_Float(
     A=&MergeHelper_FloatSwitch(
      0=&Multiply_Float(A=&MakeFloatFromVector(Vector=$User.ShoeScale_L, Channel=::NewEnumerator0), B=1)
      1=&Multiply_Float(A=&MakeFloatFromVector(Vector=$User.ShoeScale_R, Channel=::NewEnumerator0), B=1)
      2=?
      3=?
      Index=&SpawnGroup()
      Num=::NewEnumerator0
     )
     B=&FloatFromCurve(
      FloatCurve=DataInterface() {propertyValues {Curve {keys [
          {arriveTangent=-0.3053, interpMode=RCIM_Cubic, leaveTangent=-0.3053, tangentMode=RCTM_User, time=0, value=1}
          {arriveTangent=-0.5499, interpMode=RCIM_Cubic, leaveTangent=-0.5499, tangentMode=RCTM_User, time=1, value=0.5}
         ]}}}
      Scale Curve=25
     )
    )) {isSetParameters=true}
   &AddVelocity(
    Additional Quaternion Rotation=true
    Additional Yaw / Pitch / Roll:Vector3f=&MaskVectorBySpawnGroup(Masked Vector=<-80 0 -10>, Passthrough Vector=<100 0 -10>)
    Rotation Mode=::NewEnumerator3
    Rotation Quaternion:Quat4f=$Particles.SampledRotation
    Velocity=<0 -200 50>
    Velocity Speed Scale=&FN_FrontEndFloatModifier(Gameplay=0, Front End=1)
   ) {enabled=false, module=AddVelocity001}
   &AddVelocity(Velocity=<0 -300 50>)
  ]
  particleUpdateScript [
   &InheritSourceMovement(
    Applied Position Scale=&VectorFromFloat(Value=&FloatFromCurve(FloatCurve=DataInterface() {propertyValues {Curve {keys [
          {arriveTangent=-0.2727, interpMode=RCIM_Cubic, leaveTangent=-0.2727, tangentMode=RCTM_User, time=0, value=1}
          {arriveTangent=-0.5057, interpMode=RCIM_Cubic, leaveTangent=-0.5057, tangentMode=RCTM_User, time=1, value=0.14}
         ]}}}))
    ApplyForce=false
    ApplyPosition=true
    ApplyPositionFalloff=false
   ) {enabled=false}
   &SolveForcesAndVelocity(Clamp Velocity=false, Limit Acceleration=false)
   /Game/Effects/Niagara/Modules/Ribbons/FN_UpdateRibbon_v4(Enable Skeletal Position=true, SkeletalMesh:NiagaraDataInterfaceSkeletalMesh=$System.OnlyRibbons)
   /SetVariables(
    Particles.AlphaScale:NiagaraFloat=&FN_FrontEndFloatModifier(
     Gameplay=&RemapRange(
      Input Value To Remap=&VectorLength(Vector=$Particles.FN_UpdateRibbon_v4.SampledVelocity)
      Input Min=400
      Input Max=1000
      Output Min=0
      Output Max=1
     )
     Front End=1
    )
    Particles.GlowColor=<rgba 1 0.3669 0.025 1>
    Particles.LightningColor:LinearColor=<rgba 3 1.0027 0 1>
   ) {isSetParameters=true}
   &FN_RibbonMaterial(
    Color=<rgba 1.566 0.5946 0.1843 1>
    Color Pan U Speed=0.1
    Opacity Pan U Speed=1.5
    Opacity Power=0.5
    Opacity Scale U=2
    Scale Alpha:NiagaraFloat=&FloatFromCurve(
     FloatCurve=DataInterface() {propertyValues {Curve {keys [
         {arriveTangent=-1.0872, interpMode=RCIM_Cubic, leaveTangent=-1.0872, tangentMode=RCTM_User, time=0, value=0.5}
         {arriveTangent=-0.3496, interpMode=RCIM_Cubic, leaveTangent=-0.3496, tangentMode=RCTM_User, time=1, value=0}
        ]}}}
     Scale Curve=&Multiply_Float(
      A=$Particles.AlphaScale
      B=&Multiply_Float(A=$User.TrailOn, B=&FNTimeOfDayScalar(Night Scale=0.6, Morning Scale=1, Evening Scale=1))
     )
    )
   )
   /SetVariables(Particles.RibbonWidth=&Multiply_Float(
     A=$Particles.Initial.RibbonWidth
     B=&FloatFromCurve(FloatCurve=DataInterface() {propertyValues {Curve {keys [<at 0 0>,<at 0.246 1>,<at 1 1>]}}})
    )) {isSetParameters=true}
  ]
  rendererClasses [&NiagaraRibbonRendererProperties]
  renderers [
   {rendererClass=&NiagaraRibbonRendererProperties}
   {rendererClass=&NiagaraRibbonRendererProperties, rendererIndex=1}
  ]
 }
 FN_Debug {
  bEnabled=false
  emitterUpdateScript [
   &EmitterState()
   &SampleSkeletalMesh(
    AdditionalYawPitchRoll=false
    Bone Sampling Mode:ENiagaraBoneSamplingMode=::NewEnumerator4
    Skeletal Mesh:NiagaraDataInterfaceSkeletalMesh=$System.Skeletal Mesh
   )
   &SpawnBurst_Instantaneous(Spawn Count=1, Use Spawn Probability=false)
  ]
  particleSpawnScript [/Niagara/Modules/Spawn/Initialization/InitializeParticle(
    Lifetime=5
    Module.Write Color=true
    Module.Write Lifetime=true
    Module.Write Mass=true
    Module.Write Material Random=true
    Module.Write Position=false
    Module.Write Scale=false
    Module.Write SpriteRotation=false
    Module.Write SpriteSize=true
    Position=&SimulationPosition()
    Sprite Size=<100 100>
   )]
  particleUpdateScript [
   &ParticleState()
   /SetVariables(Particles.Debug:NiagaraFloat=&ClampFloat(Float=&MakeFloatFromVector(Vector=$Emitter.SampleSkeletalMesh.SampledVelocity, Channel=::NewEnumerator1))) {isSetParameters=true}
   &DynamicMaterialParameters(
    Index 0 Param 1=$Particles.Debug
    Module.Param0WriteEnabled=true
    Module.Param1WriteEnabled=true
    Module.Param2WriteEnabled=true
    Module.Param3WriteEnabled=true
   )
  ]
  rendererClasses [&NiagaraSpriteRendererProperties]
  renderers [{rendererClass=&NiagaraSpriteRendererProperties}]
 }
 Flash_Elements {
  emitterUpdateScript [
   &EmitterState()
   &SpawnRate(
    Spawn Probability=&RandomRangeFloat(Minimum=0, Maximum=1, Evaluation Type=::NewEnumerator1)
    SpawnGroup=2
    SpawnRate=15
    Use Spawn Probability=true
   )
  ]
  particleSpawnScript [
   /SetVariables(
    Particles.LightningIndex:NiagaraInt32=&RandomRangeInt(Minimum=0, Maximum=3)
    Particles.LightningLength:NiagaraFloat=&Multiply_Float(
     A=&RandomRangeFloat(Minimum=2.15, Maximum=2.75)
     B=&FN_FrontEndFloatModifier(Gameplay=1, Front End=0.45)
    )
    Particles.MirrorLightning:NiagaraBool=&RandomBool()
   ) {isSetParameters=true}
   &PartitionParticles()
   /Niagara/Modules/Spawn/Initialization/V2/InitializeParticle(
    Lifetime=&RandomRangeFloat(Minimum=0.4, Maximum=0.15)
    Mesh Scale=&Multiply_Vector(
     A=&RandomRangeVector(Minimum=<0.5 0.6 0.6>, Maximum=<1.5 1.1 1.1>)
     B=&VectorFromFloat(Value=10)
    )
    Mesh Scale Mode=::NewEnumerator4
    UsePositionOffset=false
    Write Material Random=true
   )
   &SpriteFacingAndAlignment(Sprite Alignment=$Particles.ShapeLocation.ShapeNormal, WriteAlignment=true, WriteFacing=false)
   &FN_PerParticleLOD(Enable SpriteSize Distance Scaling=true, Sprite Size Scale Far=<2.5 2.5>) {enabled=false}
   /SetVariables(Particles.SpriteSize=<1 1>, Particles.VisibilityTag=&SpawnGroup()) {isSetParameters=true}
   &ShapeLocation(
    MaskSpawnGroup=false
    Non Uniform Scale=<1 3 1>
    Offset=&MaskVector(Mask Source=1, Mask Parameter=$Particles.PartitionParticles.Partition, Mask=0, Masked Vector=<5 -8 0>, Passthrough Vector=<-4 8 0>)
    Offset Mode=::NewEnumerator0
    Override Seed=false
    Rotation Angle=0.5
    Rotation Axis=&MaskVector(Mask Source=1, Mask Parameter=$Particles.PartitionParticles.Partition, Mask=0, Masked Vector=<0 0 0>, Passthrough Vector=<0 0 0>)
    Rotation Mode=::NewEnumerator1
    Shape Origin=$Particles.SkeletalMeshLocation001.SampledPosition
    Sphere Radius=4
    UseOwnerScale=true
   )
   &InitialMeshOrientation(
    Mesh Orientation Mode=::NewEnumerator2
    Orientation Vector=&RandomRangeVector(Minimum=<1 -1 1>, Maximum=<-1 1 0>)
    Use Rotation Vector=false
   )
   /SetVariables() {isSetParameters=true}
  ]
  particleUpdateScript [
   &ParticleState()
   &SkeletalMeshLocation(
    Bone / Socket Index=$Particles.PartitionParticles.Partition
    Bone Sampling Mode=::NewEnumerator4
    MaskSpawnGroup=false
    Orientation Sampling=::NewEnumerator1
    Sampled Position Offset:NiagaraPosition=$Particles.ShapeLocation.ShapePosition
    Skeletal Mesh=$System.Skeletal Mesh
   ) {module=SkeletalMeshLocation001}
   /SetVariables(
    Particles.DynamicParametersAlt:Vector4f=&MakeVector4(
     X=&Multiply_Float(A=$Particles.LightningLength, B=&RandomRangeFloat(Minimum=1, Maximum=2))
     Y=&UniformAOrBFloat(A=1, B=2, Evaluation Type=::NewEnumerator0, Recalculate Random Each Loop=true, Override Seed=false)
     Z=&FloatFromCurve(FloatCurve=DataInterface() {propertyValues {Curve {
         defaultValue=1
         keys [
          {arriveTangent=-2.478, interpMode=RCIM_Cubic, leaveTangent=-2.478, tangentMode=RCTM_User, time=0, value=1}
          {interpMode=RCIM_Cubic, tangentMode=RCTM_User, time=1, value=0}
         ]
        }}})
     W=0
    )
    Particles.DynamicParametersAlt2:Vector4f=&MakeVector4(X=1, Y=1, Z=1, W=0)
   ) {isSetParameters=true}
   &FN_Mesh_Attributes(
    Color=&MakeLinearColorFromVectorAndFloat(
     "Vector (RGB)"=&MakeVectorFromLinearColorRGB(Color=<rgba 10 2.3821 0 1>)
     "Float (Alpha)"=&FloatFromCurve(FloatCurve=DataInterface() {propertyValues {Curve {keys [<at 0 1>,<at 1 0>]}}})
    )
    Create Color Attribute=true
    Create Scale Attribute=true
    Scale=&Multiply_Vector(
     A=&RandomRangeVector(Minimum=<1 0.6 0.6>, Maximum=<1.3 1.1 1.1>)
     B=&VectorFromFloat(Value=&Multiply_Float(
       A=0.3
       B=&MergeHelper_FloatSwitch(
        0=&Multiply_Float(A=&MakeFloatFromVector(Vector=$User.ShoeScale_L, Channel=::NewEnumerator0), B=1)
        1=&Multiply_Float(A=&MakeFloatFromVector(Vector=$User.ShoeScale_R, Channel=::NewEnumerator0), B=1)
        2=?
        3=?
        Index=$Particles.PartitionParticles.Partition
        Num=::NewEnumerator0
       )
      ))
    )
    Write to Module Namespace=true
   )
  ]
  rendererClasses [&NiagaraMeshRendererProperties]
  renderers [{rendererClass=&NiagaraMeshRendererProperties, rendererIndex=0}]
 }
 IdleCubes {
  emitterUpdateScript [
   &EmitterState(Life Cycle Mode=::NewEnumerator1, Loop Behavior=::NewEnumerator1, Loop Duration Mode=::NewEnumerator1, UseLoopDelay=false)
   &SpawnRate(
    Spawn Probability=&MakeCustomFloatFromBool(
     Bool=$User.IsSkydiving
     True Float=0
     False Float=&MakeCustomFloatFromBool(
      Bool=$User.HideFootFX
      True Float=0
      False Float=&MakeCustomFloatFromBool(Bool=$System.isRunning, True Float=0, False Float=10)
     )
    )
    SpawnRate=15
    Use Spawn Probability=true
   )
  ]
  particleSpawnScript [
   &PartitionParticles()
   &InitializeParticle(
    Color=<rgba 0.9479 0.2852 0 1>
    Lifetime=&RandomRangeFloat(Minimum=1, Maximum=1.5)
    Mesh Scale Mode=::NewEnumerator3
    Mesh Uniform Scale=&RandomRangeFloat(Minimum=0.1, Maximum=0.35)
    Position Offset=&RandomRangeVector(Minimum=<500 0 0>, Maximum=<-500 0 0>)
    UsePositionOffset=false
    Write Material Random=true
   )
   &SubUVAnimation(
    Mesh Renderer=DataInterface() {propertyValues {MeshRenderer=&self:IdleCubes_0.NiagaraMeshRendererProperties_0}}
    Output Alternate Renderer Binding=true
    Random Start Frame=true
    Renderer Type:ENiagaraMeshOrSprite=::NewEnumerator1
    SubUV Lookup Index Scale=0
    UseEndFrame=false
    UseStartFrame=false
   ) {enabled=false}
   &AddVelocity(Velocity=&RandomRangeVector(Minimum=<0 0 0>, Maximum=<1 1 6>)) {enabled=false}
   &ApplyInitialForces() {enabled=false}
   /SetVariables(Particles.ShoeCurve:NiagaraFloat=&RandomRangeFloat(Minimum=0, Maximum=1)) {enabled=false, isSetParameters=true}
  ]
  particleUpdateScript [
   &ParticleState()
   &SkeletalMeshLocation(
    Bone / Socket Index=$Particles.PartitionParticles.Partition
    Bone Sampling Mode=::NewEnumerator4
    MaskSpawnGroup=false
    Orientation Sampling=::NewEnumerator1
    Sampled Position Offset Coordinate Space:ENiagaraSkelMeshTransforms=::NewEnumerator0
    Skeletal Mesh=$System.Skeletal Mesh
   )
   &ShapeLocation(
    MaskSpawnGroup=false
    Offset=&MergeHelper_VectorSwitch(0=<1.5 0 2>, 1=<-0.5 0.5 3>, 2=?, 3=?, Index=$Particles.PartitionParticles.Partition, Num=::NewEnumerator0)
    Offset Mode=::NewEnumerator0
    Override Seed=false
    Ring Radius=&FloatFromCurve(
     FloatCurve=DataInterface() {propertyValues {Curve {defaultValue=1, keys [<at 0 1 RCIM_Cubic>,<at 1 1.5 RCIM_Cubic>]}}}
     Scale Curve=&RandomRangeFloat(Minimum=3, Maximum=4)
    )
    Rotation Coordinate Space=::Simulation
    Shape Origin=$Particles.SkeletalMeshLocation.SampledPosition
    Shape Primitive=::NewEnumerator4
    UseOwnerScale=true
   )
   &UpdateMeshOrientation(
    Rotation Rate=&RandomRangeFloat(Minimum=0.5, Maximum=0.1)
    Rotation Vector=&RandomRangeVector(Minimum=<-1 -1 -1>, Maximum=<1 1 1>)
   )
   &InheritVelocity(
    Inherited Velocity Speed Limit:NiagaraFloat=&FloatFromCurve(FloatCurve=DataInterface() {propertyValues {Curve {keys [<at 0 1>,<at 0.4886 0>]}}}, Scale Curve=5)
    Speed Limit=true
    UseSpeedThreshold=false
   ) {enabled=false}
   &ScaleMeshSize(Scale Factor=&Multiply_VectorByFloat(
     Vector=<1 1 1>
     Float=&FloatFromCurve(
      FloatCurve=DataInterface() {propertyValues {Curve {keys [
          {arriveTangent=5.9519, interpMode=RCIM_Cubic, leaveTangent=5.9519, tangentMode=RCTM_User, time=0.0003, value=0}
          {interpMode=RCIM_Cubic, tangentMode=RCTM_User, time=0.281, value=1}
          {arriveTangent=-3.6168, tangentMode=RCTM_User, time=1, value=0}
         ]}}}
      Scale Curve=&MergeHelper_FloatSwitch(
       0=&Multiply_Float(A=&MakeFloatFromVector(Vector=$User.ShoeScale_L, Channel=::NewEnumerator0), B=1)
       1=&Multiply_Float(A=&MakeFloatFromVector(Vector=$User.ShoeScale_R, Channel=::NewEnumerator0), B=1)
       2=?
       3=?
       Index=$Particles.PartitionParticles.Partition
       Num=::NewEnumerator0
      )
     )
    )) {module=ScaleMeshSize001}
   &DynamicMaterialParameters(
    Index 0 Param 1=1
    Index 0 Param 2=1
    Index 0 Param 3=1
    Module.Index0Float0Write=true
    Module.Index0Float1Write=true
    Module.Index0Float2Write=true
    Module.Index0Float3Write=true
   )
   &ScaleColor(Scale RGB=&Multiply_VectorByFloat(Vector=<1 1 1>, Float=1), ScaleA=true, ScaleRGB=true)
   &PointForce(
    Force Origin:NiagaraPosition=$Particles.SkeletalMeshLocation.SampledPosition
    Force Strength=&FloatFromCurve(
     FloatCurve=DataInterface() {propertyValues {Curve {defaultValue=1, keys [{tangentMode=RCTM_User, time=0, value=0}, {tangentMode=RCTM_User, time=1, value=1}]}}}
     Scale Curve=10
    )
    Module.Use Falloff=false
    Module.UseLinearForce=false
   ) {enabled=false}
   &SolveForcesAndVelocity(Clamp Velocity=false, Limit Acceleration=false) {enabled=false}
   &OffsetPosition(
    Position Offset=&MakeVector(
     X=0
     Y=0
     Z=&FloatFromCurve(
      FloatCurve=DataInterface() {propertyValues {Curve {defaultValue=0, keys [{tangentMode=RCTM_User, time=0, value=0}, {tangentMode=RCTM_User, time=1, value=1}]}}}
      Scale Curve=&RandomRangeFloat(Minimum=5, Maximum=8)
     )
    )
    Position Offset Coordinate Space:ENiagaraCoordinateSpace=::Local
   )
  ]
  rendererClasses [&NiagaraSpriteRendererProperties, &NiagaraMeshRendererProperties]
  renderers [
   {rendererClass=&NiagaraSpriteRendererProperties}
   {rendererClass=&NiagaraMeshRendererProperties, rendererIndex=1}
  ]
 }
 Minimal {
  bEnabled=false
  emitterUpdateScript [
   &EmitterState()
   &SpawnRate(SpawnRate=4, Use Spawn Probability=false)
   &SpawnBurst_Instantaneous(Spawn Count=2, Use Spawn Probability=false, UseLoopCountLimit=false)
   /SetVariables(Emitter.Vector 4:Vector4f=&MakeVector4(X=0, Y=0.75, Z=0, W=&Multiply_Float(A=$Emitter.Age, B=0.0001))) {isSetParameters=true, module=SetVariables_#6}
  ]
  particleSpawnScript [
   &InitializeParticle(Color Mode=::NewEnumerator0, Mesh Scale Mode=::NewEnumerator3, Mesh Uniform Scale=5, UsePositionOffset=false, Write Material Random=true)
   &PartitionParticles()
   /SetVariables(Particles.VisibilityTag=$Particles.PartitionParticles.Partition) {isSetParameters=true, module=SetVariables_#7}
   /SetVariables(Particles.LightningIndex:NiagaraInt32=&SelectIntFromArray(
     Direct Array Index=&ReturnParticleIDAttributes()
     Direct Array Mode=::NewEnumerator1
     Int Selection Array=DataInterface() {propertyValues {IntData [0,1,2,3,4,3,4,1,0,2]}}
    )) {enabled=false, isSetParameters=true, module=SetVariables_#8}
   /SetVariables() {isSetParameters=true, module=SetVariables_#9}
   /SetVariables(
    Particles.InvertLightning:NiagaraBool=&RandomBool()
    Particles.LightningIndex=1
    Particles.MirrorLightning:NiagaraBool=&RandomBool()
   ) {isSetParameters=true, module=SetVariables_#10}
  ]
  particleUpdateScript [
   &ParticleState()
   &DynamicMaterialParameters(
    Index 0 Param 1=&Multiply_Float(A=&RandomRangeFloat(Minimum=2, Maximum=7), B=2)
    Index 0 Param 2=&Divide_Float(A=1, B=1.5)
    Index 0 Param 4=&MergeHelper_FloatSwitch(0=1, 1=1, 2=1, 3=0, 4=1, Index=$Particles.LightningIndex, Num=::NewEnumerator3)
    Index 1 Param 1=&MakeFloatFromVector(Vector=$Particles.Scale, Channel=::NewEnumerator0)
    Index 1 Param 2=&MakeFloatFromVector(Vector=$Particles.Scale, Channel=::NewEnumerator1)
    Index 1 Param 3=&MakeFloatFromVector(Vector=$Particles.Scale, Channel=::NewEnumerator2)
    Index 2 Param 1=&Multiply_Float(
     A=&FloatFromCurve(FloatCurve=DataInterface() {propertyValues {Curve {defaultValue=1, keys [{tangentMode=RCTM_User, time=0, value=0}, {tangentMode=RCTM_User, time=1, value=1}]}}})
     B=&RandomRangeFloat(
      Minimum=1
      Maximum=&FloatFromCurve(
       FloatCurve=DataInterface() {propertyValues {Curve {keys [
           {arriveTangent=0.38, interpMode=RCIM_Cubic, leaveTangent=0.38, tangentMode=RCTM_User, time=0, value=0.1}
           {arriveTangent=1.1158, interpMode=RCIM_Cubic, leaveTangent=1.1158, tangentMode=RCTM_User, time=1, value=1}
          ]}}}
       CurveIndex=&RandomRangeFloat(Minimum=0, Maximum=1)
       Scale Curve=10
      )
     )
    )
    Index 2 Param 2=&FloatFromCurve(
     FloatCurve=DataInterface() {propertyValues {Curve {keys [{interpMode=RCIM_Cubic, tangentMode=RCTM_Break, time=0.1, value=1}, <at 1 0>]}}}
     Scale Curve=&RandomRangeFloat(Minimum=0, Maximum=1)
    )
    Module.Index0Param0Write=true
    Module.Index0Param1Write=true
    Module.Index0Param2Write=true
    Module.Index0Param3Write=true
    Module.Index1Param0Write=true
    Module.Index1Param1Write=true
    Module.Index1Param2Write=true
    Module.Index1Param3Write=true
    Module.Index2Param0Write=true
    Module.Index2Param1Write=true
    Module.Index2Param2Write=true
    Module.Index2Param3Write=true
    Write Parameter Index 1=true
    Write Parameter Index 2=true
   ) {enabled=false, module=DynamicMaterialParameters001}
   &DynamicMaterialParameters(
    Index 0 Param 1=13
    Index 0 Param 2=0.6
    Index 0 Param 4=&MergeHelper_FloatSwitch(0=1, 1=1, 2=1, 3=0, 4=1, Index=$Particles.LightningIndex, Num=::NewEnumerator3)
    Index 1 Param 1=&MakeFloatFromVector(Vector=$Particles.Scale, Channel=::NewEnumerator0)
    Index 1 Param 2=&MakeFloatFromVector(Vector=$Particles.Scale, Channel=::NewEnumerator1)
    Index 1 Param 3=&MakeFloatFromVector(Vector=$Particles.Scale, Channel=::NewEnumerator2)
    Index 2 Param 1=&FloatFromCurve(
     FloatCurve=DataInterface() {propertyValues {Curve {defaultValue=1, keys [{tangentMode=RCTM_User, time=0, value=0}, {tangentMode=RCTM_User, time=1, value=1}]}}}
     Scale Curve=&RandomRangeFloat(Minimum=3, Maximum=8)
    )
    Index 2 Param 2=&FloatFromCurve(FloatCurve=DataInterface() {propertyValues {Curve {defaultValue=1, keys [{tangentMode=RCTM_User, time=0, value=0}, {tangentMode=RCTM_User, time=0.3, value=1}, <at 0.7952 1>, <at 1 0>]}}})
    Module.Index0Param0Write=true
    Module.Index0Param1Write=true
    Module.Index0Param2Write=true
    Module.Index0Param3Write=true
    Module.Index1Param0Write=true
    Module.Index1Param1Write=true
    Module.Index1Param2Write=true
    Module.Index1Param3Write=true
    Module.Index2Param0Write=true
    Module.Index2Param1Write=true
    Module.Index2Param2Write=true
    Module.Index2Param3Write=true
    Write Parameter Index 1=true
    Write Parameter Index 2=true
   )
   &SkeletalMeshLocation(
    Bone / Socket Index=$Particles.Initial.VisibilityTag
    Bone Sampling Mode=::NewEnumerator4
    MaskSpawnGroup=false
    Skeletal Mesh=$System.Skeletal Mesh
   )
  ]
  rendererClasses [&NiagaraMeshRendererProperties]
  renderers [
   {rendererClass=&NiagaraMeshRendererProperties, rendererIndex=0}
   {rendererClass=&NiagaraMeshRendererProperties, rendererIndex=1}
  ]
 }
 Minimal001 {
  bEnabled=false
  emitterUpdateScript [
   &EmitterState()
   &SpawnRate(SpawnRate=4, Use Spawn Probability=false) {enabled=false}
   &SpawnBurst_Instantaneous(Spawn Count=2, Use Spawn Probability=false, UseLoopCountLimit=false)
   /SetVariables(Emitter.Vector 4:Vector4f=&MakeVector4(X=0, Y=0.75, Z=0, W=&Multiply_Float(A=$Emitter.Age, B=0.0001))) {isSetParameters=true, module=SetVariables_#6}
  ]
  particleSpawnScript [
   &InitializeParticle(Color=<rgba 23 6.9843 0 1>, Mesh Scale Mode=::NewEnumerator3, Mesh Uniform Scale=5, UsePositionOffset=false, Write Material Random=true)
   &PartitionParticles()
   /SetVariables(Particles.VisibilityTag=$Particles.PartitionParticles.Partition) {isSetParameters=true, module=SetVariables_#7}
   /SetVariables(Particles.LightningIndex:NiagaraInt32=&SelectIntFromArray(
     Direct Array Index=&ReturnParticleIDAttributes()
     Direct Array Mode=::NewEnumerator1
     Int Selection Array=DataInterface() {propertyValues {IntData [0,1,2,3,4,3,4,1,0,2]}}
    )) {enabled=false, isSetParameters=true, module=SetVariables_#8}
   /SetVariables() {isSetParameters=true, module=SetVariables_#9}
   /SetVariables(
    Particles.InvertLightning:NiagaraBool=&RandomBool()
    Particles.LightningIndex=1
    Particles.LightningLength:NiagaraFloat=&Multiply_Float(
     A=&RandomRangeFloat(Minimum=2.15, Maximum=2.75)
     B=&FN_FrontEndFloatModifier(Gameplay=1, Front End=0.45)
    )
    Particles.MirrorLightning:NiagaraBool=&RandomBool()
   ) {isSetParameters=true, module=SetVariables_#10}
  ]
  particleUpdateScript [
   &ParticleState(Kill Particles When Lifetime Has Elapsed=false)
   &DynamicMaterialParameters(
    Index 0 Param 1=&Multiply_Float(A=&RandomRangeFloat(Minimum=2, Maximum=7), B=2)
    Index 0 Param 2=&Divide_Float(A=1, B=1.5)
    Index 0 Param 4=&MergeHelper_FloatSwitch(0=1, 1=1, 2=1, 3=0, 4=1, Index=$Particles.LightningIndex, Num=::NewEnumerator3)
    Index 1 Param 1=&MakeFloatFromVector(Vector=$Particles.Scale, Channel=::NewEnumerator0)
    Index 1 Param 2=&MakeFloatFromVector(Vector=$Particles.Scale, Channel=::NewEnumerator1)
    Index 1 Param 3=&MakeFloatFromVector(Vector=$Particles.Scale, Channel=::NewEnumerator2)
    Index 2 Param 1=&Multiply_Float(
     A=&FloatFromCurve(FloatCurve=DataInterface() {propertyValues {Curve {defaultValue=1, keys [{tangentMode=RCTM_User, time=0, value=0}, {tangentMode=RCTM_User, time=1, value=1}]}}})
     B=&RandomRangeFloat(
      Minimum=1
      Maximum=&FloatFromCurve(
       FloatCurve=DataInterface() {propertyValues {Curve {keys [
           {arriveTangent=0.38, interpMode=RCIM_Cubic, leaveTangent=0.38, tangentMode=RCTM_User, time=0, value=0.1}
           {arriveTangent=1.1158, interpMode=RCIM_Cubic, leaveTangent=1.1158, tangentMode=RCTM_User, time=1, value=1}
          ]}}}
       CurveIndex=&RandomRangeFloat(Minimum=0, Maximum=1)
       Scale Curve=10
      )
     )
    )
    Index 2 Param 2=&FloatFromCurve(
     FloatCurve=DataInterface() {propertyValues {Curve {keys [{interpMode=RCIM_Cubic, tangentMode=RCTM_Break, time=0.1, value=1}, <at 1 0>]}}}
     Scale Curve=&RandomRangeFloat(Minimum=0, Maximum=1)
    )
    Module.Index0Param0Write=true
    Module.Index0Param1Write=true
    Module.Index0Param2Write=true
    Module.Index0Param3Write=true
    Module.Index1Param0Write=true
    Module.Index1Param1Write=true
    Module.Index1Param2Write=true
    Module.Index1Param3Write=true
    Module.Index2Param0Write=true
    Module.Index2Param1Write=true
    Module.Index2Param2Write=true
    Module.Index2Param3Write=true
    Write Parameter Index 1=true
    Write Parameter Index 2=true
   ) {enabled=false, module=DynamicMaterialParameters001}
   &DynamicMaterialParameters(
    Index 0 Param 1=13
    Index 0 Param 2=0.6
    Index 0 Param 4=&MergeHelper_FloatSwitch(0=1, 1=1, 2=1, 3=0, 4=1, Index=$Particles.LightningIndex, Num=::NewEnumerator3)
    Index 1 Param 1=&MakeFloatFromVector(Vector=$Particles.Scale, Channel=::NewEnumerator0)
    Index 1 Param 2=&MakeFloatFromVector(Vector=$Particles.Scale, Channel=::NewEnumerator1)
    Index 1 Param 3=&MakeFloatFromVector(Vector=$Particles.Scale, Channel=::NewEnumerator2)
    Index 2 Param 1=&FloatFromCurve(
     FloatCurve=DataInterface() {propertyValues {Curve {defaultValue=1, keys [{tangentMode=RCTM_User, time=0, value=0}, {tangentMode=RCTM_User, time=1, value=1}]}}}
     Scale Curve=&RandomRangeFloat(Minimum=3, Maximum=8)
    )
    Index 2 Param 2=&FloatFromCurve(FloatCurve=DataInterface() {propertyValues {Curve {defaultValue=1, keys [{tangentMode=RCTM_User, time=0, value=0}, {tangentMode=RCTM_User, time=0.3, value=1}, <at 0.7952 1>, <at 1 0>]}}})
    Module.Index0Param0Write=true
    Module.Index0Param1Write=true
    Module.Index0Param2Write=true
    Module.Index0Param3Write=true
    Module.Index1Param0Write=true
    Module.Index1Param1Write=true
    Module.Index1Param2Write=true
    Module.Index1Param3Write=true
    Module.Index2Param0Write=true
    Module.Index2Param1Write=true
    Module.Index2Param2Write=true
    Module.Index2Param3Write=true
    Write Parameter Index 1=true
    Write Parameter Index 2=true
   )
   /SetVariables(
    Particles.DynamicParametersAlt:Vector4f=&MakeVector4(
     X=&Multiply_Float(A=$Particles.LightningLength, B=1.2)
     Y=0.7
     Z=&FloatFromCurve(FloatCurve=DataInterface() {propertyValues {Curve {
         defaultValue=1
         keys [
          {arriveTangent=-2.478, interpMode=RCIM_Cubic, leaveTangent=-2.478, tangentMode=RCTM_User, time=0, value=1}
          {interpMode=RCIM_Cubic, tangentMode=RCTM_User, time=1, value=0}
         ]
        }}})
     W=0
    )
    Particles.DynamicParametersAlt2:Vector4f=&MakeVector4(X=1, Y=1, Z=1, W=0)
   ) {isSetParameters=true}
   &SkeletalMeshLocation(
    Bone / Socket Index=$Particles.Initial.VisibilityTag
    Bone Sampling Mode=::NewEnumerator4
    MaskSpawnGroup=false
    Skeletal Mesh=$System.Skeletal Mesh
   )
  ]
  rendererClasses [&NiagaraMeshRendererProperties]
  renderers [
   {rendererClass=&NiagaraMeshRendererProperties, rendererIndex=0}
   {rendererClass=&NiagaraMeshRendererProperties, rendererIndex=1}
  ]
 }
 RunningCubes {
  emitterUpdateScript [
   &EmitterState(Life Cycle Mode=::NewEnumerator1, Loop Behavior=::NewEnumerator1, Loop Duration Mode=::NewEnumerator1, UseLoopDelay=false)
   &SpawnRate(
    Spawn Probability=&MakeCustomFloatFromBool(
     Bool=$User.IsSkydiving
     True Float=0
     False Float=&MakeCustomFloatFromBool(
      Bool=$User.HideFootFX
      True Float=0
      False Float=&MakeCustomFloatFromBool(Bool=$System.isRunning, True Float=24, False Float=0)
     )
    )
    SpawnRate=80
    Use Spawn Probability=true
   )
  ]
  particleSpawnScript [
   &InitializeParticle(
    Color=<rgba 0.9479 0.2852 0 1>
    Lifetime=&RandomRangeFloat(Minimum=0.1, Maximum=0.17)
    Mesh Scale Mode=::NewEnumerator3
    Mesh Uniform Scale=&RandomRangeFloat(Minimum=0.4, Maximum=0.75)
    Position Offset=&RandomRangeVector(Minimum=<500 0 0>, Maximum=<-500 0 0>)
    UsePositionOffset=false
    Write Material Random=true
   )
   &SubUVAnimation(
    Mesh Renderer=DataInterface() {propertyValues {MeshRenderer=&self:RunningCubes_0.NiagaraMeshRendererProperties_0}}
    Output Alternate Renderer Binding=true
    Random Start Frame=true
    Renderer Type:ENiagaraMeshOrSprite=::NewEnumerator1
    SubUV Lookup Index Scale=0
    UseEndFrame=false
    UseStartFrame=false
   )
   &PartitionParticles()
   &AddVelocity(
    Module.Use Falloff Distance=false
    Override Seed=false
    Velocity Mode=::NewEnumerator1
    Velocity Speed=&RandomRangeFloat(Minimum=25, Maximum=10)
   )
   &ApplyInitialForces()
   &SkeletalMeshLocation(
    Bone / Socket Index=$Particles.PartitionParticles.Partition
    Bone Sampling Mode=::NewEnumerator4
    MaskSpawnGroup=false
    Skeletal Mesh=$System.Skeletal Mesh
   )
  ]
  particleUpdateScript [
   &ParticleState()
   &UpdateMeshOrientation(
    Rotation Rate=&RandomRangeFloat(Minimum=5, Maximum=2)
    Rotation Vector=&RandomRangeVector(Minimum=<-1 -1 -1>, Maximum=<1 1 1>)
   )
   &InheritVelocity(
    Inherited Velocity Speed Limit:NiagaraFloat=&FloatFromCurve(FloatCurve=DataInterface() {propertyValues {Curve {keys [<at 0 1>,<at 0.4886 0>]}}}, Scale Curve=100)
    Speed Limit=true
    UseSpeedThreshold=false
   )
   &ScaleMeshSize(Scale Factor=&Multiply_VectorByFloat(
     Vector=<1 1 1>
     Float=&FloatFromCurve(
      FloatCurve=DataInterface() {propertyValues {Curve {keys [{interpMode=RCIM_Cubic, tangentMode=RCTM_User, time=0, value=1}, {arriveTangent=-3.6168, tangentMode=RCTM_User, time=1, value=0}]}}}
      Scale Curve=&MergeHelper_FloatSwitch(
       0=&Multiply_Float(A=&MakeFloatFromVector(Vector=$User.ShoeScale_L, Channel=::NewEnumerator0), B=1)
       1=&Multiply_Float(A=&MakeFloatFromVector(Vector=$User.ShoeScale_R, Channel=::NewEnumerator0), B=1)
       2=?
       3=?
       Index=$Particles.PartitionParticles.Partition
       Num=::NewEnumerator0
      )
     )
    )) {module=ScaleMeshSize001}
   &DynamicMaterialParameters(
    Index 0 Param 1=1
    Index 0 Param 2=1
    Index 0 Param 3=1
    Module.Index0Float0Write=true
    Module.Index0Float1Write=true
    Module.Index0Float2Write=true
    Module.Index0Float3Write=true
   )
   &ScaleColor(Scale RGB=&Multiply_VectorByFloat(Vector=<1 1 1>, Float=1), ScaleA=true, ScaleRGB=true)
   &SolveForcesAndVelocity(Clamp Velocity=false, Limit Acceleration=false)
  ]
  rendererClasses [&NiagaraSpriteRendererProperties, &NiagaraMeshRendererProperties]
  renderers [
   {rendererClass=&NiagaraSpriteRendererProperties}
   {rendererClass=&NiagaraMeshRendererProperties, rendererIndex=1}
  ]
 }
}
summary {
 systemName=NS_Shoe_VoyagerCuboid_Idle
 emitters [
  {bEnabled=false, emitterName=FN_AttachedRibbon001, rendererClasses [&NiagaraRibbonRendererProperties]}
  {emitterName=RunningCubes, rendererClasses [&NiagaraSpriteRendererProperties, &NiagaraMeshRendererProperties]}
  {emitterName=IdleCubes, rendererClasses [&NiagaraSpriteRendererProperties, &NiagaraMeshRendererProperties]}
  {bEnabled=false, emitterName=Minimal, rendererClasses [&NiagaraMeshRendererProperties]}
  {bEnabled=false, emitterName=Minimal001, rendererClasses [&NiagaraMeshRendererProperties]}
  {emitterName=Flash_Elements, rendererClasses [&NiagaraMeshRendererProperties]}
  {bEnabled=false, emitterName=FN_Debug, rendererClasses [&NiagaraSpriteRendererProperties]}
  {emitterName=FN_AnimTrail002, rendererClasses [&NiagaraRibbonRendererProperties]}
 ]
 userVariables [
  User.bIsFrontEnd {type=NiagaraBool}
  User.bIsFrontEndPreview {type=NiagaraBool}
  User.bisLit {type=NiagaraBool}
  User.bisPrimaryPreviewItem {type=NiagaraBool}
  User.HideFootFX {type=NiagaraBool}
  User.isOffPawnPreview {type=NiagaraBool}
  User.IsSkydiving {type=NiagaraBool}
  User.isStage2 {type=NiagaraBool}
  User.isStage3 {type=NiagaraBool}
  User.RibbonIDOffset {type=NiagaraInt32}
  User.ShoeScale_L {type=Vector3f, description="", defaultValue {struct=Vector3f, value=<1 1 1>}}
  User.ShoeScale_R {type=Vector3f, description="", defaultValue {struct=Vector3f, value=<1 1 1>}}
  User.SkeletalMesh {
   type=NiagaraDataInterfaceSkeletalMesh
   defaultValue {
    struct=NiagaraExt_VariableValue_DataInterface
    value {dataInterface=&self:NiagaraDataInterfaceSkeletalMesh_1}
   }
  }
  User.TrailOn {type=NiagaraFloat, defaultValue=1, description=""}
 ]
}
