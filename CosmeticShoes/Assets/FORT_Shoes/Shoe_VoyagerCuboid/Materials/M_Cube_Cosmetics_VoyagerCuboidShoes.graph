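# Masked, two-sided Default Lit material for the VoyagerCuboid cosmetic shoes, built as a chain of shared material functions passing material attributes.
# The CharacterParent 2023 PreFX base feeds the SpacePlungeCube effect borrowed from a backpack, with a constant mask of one.
# MF_Sock then applies using the base's Tech Art Mask, before CharacterParent PostFX and a final shoe hologram function.
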Material /CosmeticShoes/Assets/FORT_Shoes/Shoe_VoyagerCuboid/Materials/M_Cube_Cosmetics_VoyagerCuboidShoes
graph {
 properties {
  bUseMaterialAttributes=true
  bUsedWithInstancedStaticMeshes=true
  bUsedWithMeshDeformer=true
  bUsedWithSkeletalMesh=true
  blendMode=BLEND_Masked
  pixelDepthOffsetMode=PDOM_Legacy
  twoSided=true
  usedShadingModels=Default Lit
 }
 nodes [
  n1 = MaterialFunctionCall() {materialFunction=&MF_CharacterParent_2023_PreFX}
  n2 = MaterialFunctionCall!(In=@n1.Result, Mask=Constant() {r=1}) {materialFunction=/BRCosmetics/Accessories/FORT_Backpacks/Backpack_M_MED_SpacePlunge/Material/MF_SpacePlungeCube}
  MaterialFunctionCall!(In=MaterialFunctionCall(
    Material=MaterialFunctionCall!(InOpacity, Material=@n2.Result, Mask=@n1.Tech Art Mask) {materialFunction=/Game/Global/Materials/Cosmetics/Kicks/Socks/Materials/MF_Sock}.Result
    M=@n1.M
    Pre Skinned Local Position=@n1.M
    Distance Blend=@n1.M
    Fresnel=@n1.M
   ) {materialFunction=&MF_CharacterParent_2023_PostFX}.Result) {materialFunction=/Game/Global/Materials/Cosmetics/Kicks/Shoes/Materials/MF_TechartShoeHologram}
 ]
}
