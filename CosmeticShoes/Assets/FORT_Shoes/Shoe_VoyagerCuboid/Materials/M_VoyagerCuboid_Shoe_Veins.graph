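# Translucent skeletal mesh material for the VoyagerCuboid cosmetic shoe, adding lightning veins to the shared 2023 character material.
# The character PreFX function supplies base attributes, which the lightning veins function modifies using the VeinsMask texture parameter.
# The result passes through the sock function, masked by the PreFX Tech Art Mask, then through character PostFX and finally the shoe hologram function.
# It uses material attributes throughout, with shading model set by expression.
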Material /CosmeticShoes/Assets/FORT_Shoes/Shoe_VoyagerCuboid/Materials/M_VoyagerCuboid_Shoe_Veins
graph {
 properties {
  bAutomaticallySetUsageInEditor=false
  bUseMaterialAttributes=true
  bUsedWithMeshDeformer=true
  bUsedWithMorphTargets=true
  bUsedWithSkeletalMesh=true
  blendMode=BLEND_Translucent
  pixelDepthOffsetMode=PDOM_Legacy
  shadingModel=MSM_FromMaterialExpression
  usedShadingModels=Default Lit | Subsurface | Clear Coat | Subsurface Profile | Cloth
 }
 nodes [
  n1 = MaterialFunctionCall() {materialFunction=&MF_CharacterParent_2023_PreFX}
  n2 = TextureSampleParameter2D() {
   parameterName=VeinsMask
   samplerType=SAMPLERTYPE_LinearColor
   texture=/CosmeticShoes/Assets/FORT_Shoes/Shoe_VoyagerCuboid/Textures/T_VoyagerCuboid_StaticLightningShoes
  }
  n3 = MaterialFunctionCall!(In_MaterialAttributes=@n1.Result, In_Mask=@n2.RGBA) {materialFunction=&/MF_LightningVeins_Shoes}
  MaterialFunctionCall!(In=MaterialFunctionCall(
    Material=MaterialFunctionCall!(InOpacity, Material=@n3.Result, Mask=@n1.Tech Art Mask) {materialFunction=/Game/Global/Materials/Cosmetics/Kicks/Socks/Materials/MF_Sock}.Result
    M=@n1.M
    Pre Skinned Local Position=@n1.M
    Distance Blend=@n1.M
    Fresnel=@n1.M
   ) {materialFunction=&MF_CharacterParent_2023_PostFX}.Result) {materialFunction=/Game/Global/Materials/Cosmetics/Kicks/Shoes/Materials/MF_TechartShoeHologram}
 ]
}
